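# Masked crystal material with thin-film rainbow tint, parallax sparkles and sphere-masked two-color emissive glow, usable on instanced and Niagara meshes.
# A time-sampled blink texture drives a height-lerped rim-glow pulse, and emissive is tinted by time of day.
# Temporal dither feeds Pixel Depth Offset, and normals pass through a tangent-to-world function.
# World Position Offset adds hit-and-shrink effects to a normal-direction blink displacement.
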Material /Game/Environments/Apollo/Sets/Quest/Props/Materials/M_CrystalsDither
graph {
 properties {
  bAutomaticallySetUsageInEditor=false
  bTangentSpaceNormal=false
  bUseMaterialAttributes=true
  bUsedWithInstancedStaticMeshes=true
  bUsedWithNiagaraMeshParticles=true
  blendMode=BLEND_Masked
  pixelDepthOffsetMode=PDOM_Legacy
  usedShadingModels=Default Lit
 }
 nodes [
  n1 = TextureSampleParameter2D(UVs=Multiply(A=Time(), B=ScalarParameter() {defaultValue=1, group=Blink, parameterName=BlinkTime}) {constB=0.25}) {
   group=Blink
   parameterName=BlinkTimeMask
   samplerType=SAMPLERTYPE_Grayscale
   texture=/Game/Environments/Apollo/Sets/Quest/Props/Textures/CrysalTime
  }
  Reroute(None=@n1.R)
  n2 = TextureSample(UVs=DotProduct(A=CameraVectorWS(), B=PixelNormalWS())) {samplerSource=SSM_Clamp_WorldGroupSettings, texture=/Game/Weapons/WeaponSkins/Wraps/Textures/T_Grad_Rainbow2}
  n3 = TextureSampleParameter2D() {
   group=Textures
   parameterName=Normals
   samplerType=SAMPLERTYPE_Normal
   texture=/Game/Environments/Apollo/Sets/DesertPoint/Crystals/Textures/Shards_AB_N
  }
  n4 = Transform(None=@n3.RGB)
  n5 = OneMinus(None=Fresnel(Normal=@n4) {baseReflectFraction=0.3, exponent=3})
  n6 = Multiply(A=Multiply(A=@n2.RGB, B=ScalarParameter() {defaultValue=0.07, parameterName=ThinFilm}), B=@n5)
  n7 = TextureSampleParameter2D() {
   group=Textures
   parameterName=SpecularMasks
   samplerType=SAMPLERTYPE_Masks
   texture=/Game/Environments/Apollo/Sets/DesertPoint/Crystals/Textures/Shards_AB_M
  }
  n8 = Multiply(A=Multiply(A=@n5, B=VectorParameter() {group=DiffuseColors, parameterName=Color2, defaultValue=<rgba 0.3738 0.5838 0.9792 0>}.RGB), B=@n7.R)
  n9 = WorldPosition()
  n10 = ActorPositionWS()
  n11 = Multiply(A=ObjectRadius(), B=ScalarParameter() {defaultValue=0.8, group=EmissiveColors, parameterName=Radius})
  n12 = ScalarParameter() {defaultValue=-0.2381, group=EmissiveColors, parameterName=Hardness}
  n13 = Saturate(None=OneMinus(None=SphereMask(A=@n9.XYZ, B=@n10, Radius=@n11, Hardness=@n12)))
  n14 = TextureCoordinate()
  n15 = Divide(A=ObjectRadius(), B=ScalarParameter() {defaultValue=800, group=Details, parameterName=ObjectRadius}) {constB=512}
  n16 = BumpOffset(
   Coordinate=Multiply(A=Multiply(A=@n14, B=@n15), B=ScalarParameter() {defaultValue=12, group=Details, parameterName=Sparkles})
   Height=Multiply(A=@n7.G) {constB=-1}
  )
  n17 = Multiply(
   A=@n13
   B=Multiply(
    A=OneMinus(None=@n7.G)
    B=TextureSample(UVs=@n16) {samplerType=SAMPLERTYPE_Masks, texture=/Game/Environments/Apollo/Sets/DesertPoint/Impact/NewSand_M}.G
   )
  )
  n18 = Power(Base=DotProduct(A=CameraVectorWS(), B=@n4))
  n19 = Add(A=Multiply(A=@n6) {constB=0.5}, B=Multiply(A=Multiply(A=@n2.RGB, B=@n17), B=Saturate(None=@n18)))
  n20 = Constant() {r=3}
  n21 = VectorParameter() {group=EmissiveColors, parameterName=Ecolor, defaultValue=<rgba 0.4557 0.2641 0.2891 0>}
  n22 = VectorParameter() {group=EmissiveColors, parameterName=Ecolor2, defaultValue=<rgba 0.3691 0.1186 0.5391 0>}
  n23 = Multiply(A=Add(A=Power(Base=@n7.R, Exp=@n20), B=@n7.G), B=Multiply(A=@n13, B=LinearInterpolate(A=@n21.RGB, B=@n22.RGB, Alpha=@n13)))
  n24 = GetMaterialAttributes(None=MakeMaterialAttributes!(
    BaseColor=Add(A=@n6, B=@n8)
    Metallic=Constant() {r=0.5238}
    Specular=ScalarParameter() {defaultValue=0.1, group=Textures, parameterName=S}
    Roughness=ScalarParameter() {defaultValue=0.2, group=Textures, parameterName=R}
    Anisotropy
    EmissiveColor=Add(A=@n19, B=@n23)
    Opacity
    OpacityMask
    Normal=@n3.RGB
    Tangent
    WorldPositionOffset
    SubsurfaceColor
    ClearCoat
    ClearCoatRoughness
    AmbientOcclusion
    Refraction
    CustomizedUVs_0
    CustomizedUVs_1
    CustomizedUVs_2
    CustomizedUVs_3
    CustomizedUVs_4
    CustomizedUVs_5
    CustomizedUVs_6
    CustomizedUVs_7
    PixelDepthOffset
    ShadingModel
    Displacement
   )) {attributeGetTypes [#1]}
  n25 = VectorParameter() {group=Blink, parameterName=BlinkGlow, defaultValue=<rgba 2 0 0 0>}
  n26 = DotProduct(A=CameraVectorWS(), B=VertexNormalWS())
  n27 = OneMinus(None=@n26)
  n28 = ScalarParameter() {group=Blink, parameterName=BlinkGlow Invert}
  n29 = Saturate(None=Power(Base=LinearInterpolate(A=@n26, B=@n27, Alpha=@n28), Exp=ScalarParameter() {defaultValue=3, group=Blink, parameterName=BlinkGlow Power}))
  n30 = Add(A=@n24.Emissive Color, B=Multiply(A=@n24.Emissive Color, B=LinearInterpolate(A=@n24.Emissive Color, B=@n25.RGB, Alpha=@n29) {constA=1}))
  n31 = OneMinus(None=NamedRerouteDeclaration(None=@n1.R) {name=BlinkTime, nodeColor=<rgba 1 0.0469 0 1>})
  n32 = TextureSampleParameter2D() {
   group=Textures
   parameterName=SpecularMasks
   samplerType=SAMPLERTYPE_Masks
   texture=&Shards_AB_M
  }
  n33 = Constant() {r=0.5}
  n34 = GetMaterialAttributes(None=BlendMaterialAttributes(
    A=SetMaterialAttributes(MaterialAttributes=@n24.MaterialAttributes, Emissive Color=@n30) {attributeSetTypes [#1]}
    B=@n24.MaterialAttributes
    Alpha=MaterialFunctionCall(Transition Phase=@n31, Height Texture=@n32.R, Contrast=@n33) {materialFunction=&HeightLerp}.Alpha
   ) {vertexAttributeBlendType=UseA}) {attributeGetTypes [#1]}
  n35 = VectorParameter() {group=TOD, parameterName=EmissiveMorning, sortPriority=1, defaultValue=<rgba 1 0.4297 0.5294 1>}
  n36 = VectorParameter() {group=TOD, parameterName=EmissiveDay, sortPriority=2, defaultValue=<rgba 1 1 1 1>}
  n37 = VectorParameter() {group=TOD, parameterName=EmissiveEvening, sortPriority=3, defaultValue=<rgba 0.5208 0.285 0.1768 1>}
  n38 = VectorParameter() {group=TOD, parameterName=EmissiveNight, sortPriority=4, defaultValue=<rgba 0.2726 0.3026 0.3333 1>}
  n39 = SetMaterialAttributes(
   MaterialAttributes=@n34.MaterialAttributes
   Emissive Color=Multiply(
    A=@n34.Emissive Color
    B=MaterialFunctionCall(Color_Morning=@n35.RGB, Color_Day=@n36.RGB, Color_Evening=@n37.RGB, Color_Night=@n38.RGB) {materialFunction=&Color_Time_of_Day}.Result
   )
  ) {attributeSetTypes [#1]}
  n40 = Multiply(
   A=MaterialFunctionCall() {materialFunction=&DitherTemporalAA}.Result
   B=ScalarParameter() {defaultValue=15, parameterName=PDO}
  )
  n41 = VectorParameter() {bUseCustomPrimitiveData=true, group=z_code, parameterName=WSDeformationCenter, primitiveDataIndex=20}
  n42 = ScalarParameter() {group=z_code, parameterName=Code_DestructionAmount}
  n43 = StaticBoolParameter() {defaultValue=true, group=z_code, parameterName=Code_UseHitEffects}
  n44 = ScalarParameter() {bUseCustomPrimitiveData=true, defaultValue=1, group=z_code, parameterName=WSDamageRadius, primitiveDataIndex=26}
  n45 = ScalarParameter() {bUseCustomPrimitiveData=true, group=z_code, parameterName=DistanceCullFadeAnimationOffset, primitiveDataIndex=12}
  n46 = Multiply(A=NamedRerouteUsage(), B=VertexNormalWS())
  n47 = ScalarParameter() {defaultValue=1, group=Blink, parameterName=BlinkGeo}
  SetMaterialAttributes!(
   MaterialAttributes=MaterialFunctionCall!(Material=SetMaterialAttributes!(MaterialAttributes=@n39, Pixel Depth Offset=@n40) {attributeSetTypes [#2]}) {materialFunction=/Game/Athena/Apollo/Maps/Landscape/VT/MaterialFunctions/TransformTangentToWorld}.Result
   World Position Offset=Add(
    A=MaterialFunctionCall!(
     WSDeformationCenter=@n41.RGB
     Code_DestructionAmount=@n42
     Code_UseUvForDamage
     Code_UseHitEffects=@n43
     Code_Emissive
     WSDamageRadius=@n44
     CrackDamageShader
     Diffuse
     ScalarModDamageCracks
     ScalarOffsetforDamageCracks
     ExposeLerpFunctionForInvertingTheHitDistanceCalc
     InvertDamageFalloff
     CrackWorldPosition
     WorldPosition
     DistanceCullFadeAnimationOffset=@n45
    ) {materialFunction=/Game/Packages/jl_tests/Fortnite_HitEffects}.Hit And Shrink WPO
    B=Multiply(A=@n46, B=@n47)
   )
  ) {attributeSetTypes [#3]}
 ]
}
