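# Suburban TV prop material for a Quest set that shows an event countdown on its emissive screen, masked by vertex color red.
# MPC_Countdown drives visibility and picks one of four vertical steps in a countdown texture, layered with a second, sphere-masked countdown texture.
# Scrolling scanlines and a radial vignette shape the tinted glow.
# The IsDamaged static switch distorts screen UVs with the normal map and darkens emission through a diffuse-green mask.
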Material /Game/Environments/Apollo/Sets/Quest/Props/Materials/M_Quest_TV_Countdown
graph {
 properties {bUseMaterialAttributes=true, pixelDepthOffsetMode=PDOM_Legacy, usedShadingModels=Default Lit}
 nodes [
  Multiply()
  n1 = TextureSampleParameter2D() {parameterName=Diffuse, texture=/Game/Environments/Apollo/Sets/Suburban/Props/Textures/T_Suburban_TV_01_D}
  n2 = Saturate(None=LinearInterpolate(
    A=ScalarParameter() {defaultValue=2.0158, parameterName=DamageMask1}
    B=ScalarParameter() {defaultValue=-49.8817, parameterName=DamageMask2}
    Alpha=@n1.G
   ))
  Reroute(None=@n2)
  ScalarParameter() {defaultValue=1, parameterName=DaysLeft}
  ScalarParameter() {parameterName=OnOFF}
  n3 = TextureSampleParameter2D() {parameterName=Specular, samplerType=SAMPLERTYPE_Masks, texture=/Game/Environments/Apollo/Sets/Suburban/Props/Textures/T_Suburban_TV_01_S}
  n4 = TextureSampleParameter2D() {parameterName=Normal, samplerType=SAMPLERTYPE_Normal, texture=/Game/Environments/Apollo/Sets/Suburban/Props/Textures/T_Suburban_TV_01_N}
  n5 = SetMaterialAttributes!(MaterialAttributes, Base Color=@n1.RGB, Metallic=@n3.G, Specular=@n3.G, Roughness=@n3.G, Normal=@n4.RGB) {attributeSetTypes [
    #1
    #2
    #3
    #4
    #5
   ]}
  n6 = LinearInterpolate(
   A=ScalarParameter() {defaultValue=2, parameterName=ScanlineBrightness}
   Alpha=MaterialFunctionCall!(
    ScanLine_Tiling_1=ScalarParameter() {defaultValue=3, parameterName=ScanlineTiling}
    Scanline_PannerSpeed_1=ScalarParameter() {defaultValue=0.25, parameterName=ScanlineSpeed}
    ScanLine_Tiling=ScalarParameter() {defaultValue=8, parameterName=ScanlineTiling2}
    Scanline_PannerSpeed=ScalarParameter() {defaultValue=-0.5, parameterName=ScanlineSpeed2}
    UVs
   ) {materialFunction=/Game/Weapons/WeaponSkins/MaterialFunction/ScanLines}.Result
  )
  n7 = TextureCoordinate() {coordinateIndex=1}
  n8 = Constant2Vector() {g=0.5, r=0.5}
  n9 = ScalarParameter() {defaultValue=0.5733, parameterName=Radius}
  n10 = Constant()
  n11 = ComponentMask(None=@n4.RGB) {g=true, r=true}
  n12 = ScalarParameter() {defaultValue=0.0533, parameterName=Distort}
  n13 = TextureCoordinate() {coordinateIndex=1}
  n14 = StaticSwitchParameter(True=Add(A=Multiply(A=@n11, B=@n12), B=@n13), False=@n13) {parameterName=IsDamaged}
  n15 = Add(A=@n14, B=AppendVector(A=ScalarParameter() {parameterName=xOffset}, B=ScalarParameter() {defaultValue=0.02, parameterName=yOffset}))
  n16 = AppendVector(A=ScalarParameter() {defaultValue=0.7943, parameterName=x}, B=ScalarParameter() {defaultValue=0.75, parameterName=y})
  n17 = ScalarParameter() {parameterName=xOffsetMask}
  n18 = ScalarParameter() {defaultValue=-0.381, parameterName=yOffsetMask}
  n19 = MaterialFunctionCall(
   UVs=Add(A=@n14, B=AppendVector(A=@n17, B=@n18))
   Texture Scale=AppendVector(A=ScalarParameter() {defaultValue=1, parameterName=xMask}, B=ScalarParameter() {defaultValue=1, parameterName=yMask})
  ) {materialFunction=&ScaleUVsByCenter}
  n20 = Constant2Vector() {g=0.5, r=0.5}
  n21 = Constant() {r=0.5}
  n22 = Constant() {r=0.3}
  n23 = Multiply(
   A=TextureSample(UVs=MaterialFunctionCall(UVs=@n15, Texture Scale=@n16) {materialFunction=&ScaleUVsByCenter}.UVs) {samplerType=SAMPLERTYPE_Masks, texture=/Game/Environments/Apollo/Sets/Quest/Props/Textures/T_Quest_CountDown_02}.R
   B=OneMinus(None=SphereMask(A=@n19.UVs, B=@n20, Radius=@n21, Hardness=@n22))
  )
  n24 = Add(A=@n14, B=AppendVector(A=ScalarParameter() {parameterName=xOffset2}, B=ScalarParameter() {defaultValue=0.115, parameterName=yOffset2}))
  n25 = ComponentMask(None=@n24) {g=true}
  n26 = Floor(None=Multiply(A=OneMinus(None=@n25)) {constB=4})
  n27 = Constant() {r=1}
  n28 = Round(None=Multiply(B=CollectionParameter() {collection=/Game/Athena/Events/Generic/CountdownTimer/Materials/MPC_Countdown, parameterName=ProgressNormalized}) {constA=4})
  n29 = AppendVector(A=Constant(), B=Multiply(A=Divide(A=@n27) {constB=4}, B=Subtract(A=@n28)))
  n30 = Multiply(
   A=Saturate(None=OneMinus(None=@n26))
   B=TextureSample(UVs=Subtract(A=@n24, B=@n29)) {samplerType=SAMPLERTYPE_Masks, texture=/Game/Environments/Apollo/Sets/Quest/Props/Textures/T_Quest_CountDown_01}.R
  )
  n31 = Add(
   A=Multiply(A=SphereMask(A=@n7, B=@n8, Radius=@n9, Hardness=@n10), B=ScalarParameter() {defaultValue=1.5, parameterName=Vignette2})
   B=Multiply(
    A=Add(A=@n23, B=@n30)
    B=CollectionParameter() {collection=&MPC_Countdown, parameterName=ShowCountdown}
   )
  )
  n32 = Multiply(
   A=VertexColor().R
   B=LinearInterpolate(
    A=Constant3Vector()
    B=VectorParameter() {parameterName=Emissive_Color, defaultValue=<rgba 0.1075 0.1571 0.2792 1>}.RGB
    Alpha=Multiply(A=@n6, B=@n31)
   )
  )
  n33 = LinearInterpolate(B=ScalarParameter() {defaultValue=0.3581, parameterName=DamageDarken}, Alpha=@n2) {constA=1}
  SetMaterialAttributes!(MaterialAttributes=@n5, Emissive Color=StaticSwitchParameter(True=Multiply(A=@n32, B=@n33), False=@n32) {parameterName=IsDamaged}) {attributeSetTypes [#6]}
 ]
}
