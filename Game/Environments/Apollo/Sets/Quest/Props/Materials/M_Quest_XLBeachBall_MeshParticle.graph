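# Masked, lit material for Niagara mesh particles of the Apollo quest XL beach ball prop.
# The diffuse texture is doubled in brightness, alongside a specular mask texture.
# The normal map runs through FlattenNormal with a Flatness parameter, whose negative default exaggerates the surface detail instead of flattening it.
# A textureless Bayer dither function is fed a constant input rather than particle-driven fade.
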Material /Game/Environments/Apollo/Sets/Quest/Props/Materials/M_Quest_XLBeachBall_MeshParticle
graph {
 properties {
  bAutomaticallySetUsageInEditor=false
  bUsedWithNiagaraMeshParticles=true
  blendMode=BLEND_Masked
  translucencyLightingMode=TLM_Surface
  usedShadingModels=Default Lit
 }
 nodes [
  Constant()
  MaterialFunctionCall(
   Normal=TextureSample() {samplerType=SAMPLERTYPE_Normal, texture=/Game/Environments/Apollo/Sets/Quest/Props/Textures/T_Quest_XLBeachBall_01_N}.RGB
   Flatness=ScalarParameter() {defaultValue=-5, parameterName="Flatness (S)"}
  ) {materialFunction=&FlattenNormal}
  MaterialFunctionCall!(In=Constant() {r=1}, Iterations, Scale) {materialFunction=/Game/Packages/Fortress_SharedMaterials/Base_Material_Functions/MF_DitherBayerNoTex}
  Multiply(
   A=Constant3Vector() {constant=<rgba 2 2 2 0>}
   B=TextureSample() {texture=/Game/Environments/Apollo/Sets/Quest/Props/Textures/T_Quest_XLBeachBall_01_D}.RGB
  )
  TextureSample() {samplerType=SAMPLERTYPE_Masks, texture=/Game/Environments/Apollo/Sets/Quest/Props/Textures/T_Quest_XLBeachBall_01_S}
 ]
}
